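; A_MusicPlayer is a Harmonix example actor that plays a MetaSound song in sync with a Music Clock component.
; On BeginPlay it connects the clock to the MetaSound on its Audio component.
; If that succeeds and AutoStart is set, it starts the clock, plays audio, then passes the song patch and MIDI file as parameters.
; StartSong and StopSong only play or stop the audio component.
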(event EventBeginPlay
  (bind _returnvalue (MusicClock|ConnecttoMetasoundonAudioComponent (Variables|Default|GetMusicClock) (Variables|Default|GetAudio)))
  (if _returnvalue
    (if (Variables|Default|GetAutoStart)
      (MusicClock|Start (Variables|Default|GetMusicClock))
      (Audio|Components|Audio|Play (Variables|Default|GetAudio) 0.0)
      (Audio|Parameter|SetObjectParameter (Variables|Default|GetAudio) "Patch" (Variables|Default|GetSongPatch))
      (Audio|Parameter|SetObjectParameter (Variables|Default|GetAudio) "MIDI File" (Variables|Default|GetSongMIDI)))))

(event Collision|EventActorBeginOverlap
  (params (OtherActor :type Actor%20Object%20Reference)))

(event EventTick
  (params (DeltaSeconds :type Float%20%28single-precision%29)))

(fn StartSong
  (Audio|Components|Audio|Play (Variables|Default|GetAudio) 0.0))

(fn StopSong
  (Audio|Components|Audio|Stop (Variables|Default|GetAudio)))

(fn ConstructionScript)
