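; A simple Harmonix example actor that provides a music clock driven by wall-clock time over a MIDI tempo map.
; On BeginPlay, when AutoStart is set, it connects its MusicClock component to the wall clock using the TempoMap variable and starts it.
; A GetClockComponent function exposes the clock component to other objects.
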(event EventBeginPlay
  (if (Variables|Default|GetAutoStart)
    (MusicClock|ConnecttoWallClockforMidi (Variables|Default|GetMusicClock) (Variables|Default|GetTempoMap))
    (MusicClock|Start (Variables|Default|GetMusicClock))))

(fn GetClockComponent
  (return (Variables|Default|GetMusicClock)))

(fn ConstructionScript)
